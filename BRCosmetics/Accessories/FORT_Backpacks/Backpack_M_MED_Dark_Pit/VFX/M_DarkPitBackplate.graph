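# Unlit masked material for Niagara sprite and mesh particles on the Dark Pit backpack's backplate VFX, drawing an animated, tinted, noise-distorted mask shape.
# Panning noise textures, on a shared Master Time clock, warp the mask UVs with side-to-middle vertical offsets, optional flame-lick erosion and custom rotation.
# A colored glow texture drives emissive.
# The main mask is shaped by particle alpha, an edge cutoff, optional view-angle fade and non-mobile depth fade.
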Material /BRCosmetics/Accessories/FORT_Backpacks/Backpack_M_MED_Dark_Pit/VFX/M_DarkPitBackplate
graph {
 properties {
  bUsedWithNiagaraMeshParticles=true
  bUsedWithNiagaraSprites=true
  blendMode=BLEND_Masked
  pixelDepthOffsetMode=PDOM_Legacy
  shadingModel=MSM_Unlit
  usedShadingModels=Unlit
 }
 nodes [
  n1 = ScalarParameter() {defaultValue=0.5, parameterName=Tex Rotation}
  n2 = MaterialFunctionCall(UVs=TextureCoordinate(), "Rotation Angle (0-1)"=@n1) {materialFunction=&CustomRotator}
  n3 = MaterialFunctionCall(UVs=TextureCoordinate(), Texture Scale=AppendVector(A=Constant() {r=1}, B=ScalarParameter() {defaultValue=1, parameterName=Texture Scale Y})) {materialFunction=&ScaleUVsByCenter}
  n4 = ComponentMask(None=@n3.UVs) {r=true}
  n5 = TextureSampleParameter2D() {parameterName=Distortion Distortion Texture, texture=/Game/Global/Textures/Default/Noise/3xPerlin_M}
  n6 = Add(A=TextureCoordinate(), B=Multiply(A=ComponentMask(None=@n5.RGB) {g=true, r=true}, B=Constant() {r=0.05}))
  n7 = DynamicParameter() {paramNames [ErosionMult, ErosionPower, TextureOffsetX, TextureOffsetY]}
  n8 = LinearInterpolate(
   A=@n6
   B=Add(A=@n6, B=AppendVector(A=@n7.TextureOffsetX, B=@n7.TextureOffsetX))
   Alpha=ScalarParameter() {parameterName=UseDynParmsForTextureOffset, sliderMax=1}
  )
  n9 = TextureSampleParameter2D(UVs=Panner(Coordinate=@n8, Time=NamedRerouteUsage()) {speedX=-0.05, speedY=0.25}) {
   parameterName="Distortion Texture (add)"
   samplerType=SAMPLERTYPE_LinearGrayscale
   texture=/BRCosmetics/Athena/Effects/Particles/Skydiving/NS_Contrail_DarkAndLight/T_DarkAndLight_Noise
  }
  n10 = Multiply(
   A=@n9.R
   B=LinearInterpolate(
    A=ScalarParameter() {defaultValue=0.2, parameterName=Noise Str 1}
    B=ScalarParameter() {defaultValue=0.2, parameterName=Noise Str 1 - 2}
    Alpha=MaterialFunctionCall() {materialFunction=&LinearGradient}.VGradient
   )
  )
  n11 = MaterialFunctionCall(UV Channel=TextureCoordinate() {uTiling=0.5}) {materialFunction=&LinearGradient}
  n12 = LinearInterpolate(
   A=ScalarParameter() {defaultValue=-0.15, parameterName=Side Offset}
   B=ScalarParameter() {defaultValue=0.15, parameterName=Middle Offset}
   Alpha=Power(Base=Sine(None=@n11.UGradient)) {constExponent=1}
  ) {constA=-0.15, constB=0.15}
  n13 = Add(A=ComponentMask(None=@n3.UVs) {g=true}, B=ScalarParameter() {defaultValue=-0.1, parameterName=Overall vertical offset})
  n14 = TextureSampleParameter2D(UVs=Panner(Coordinate=@n8, Time=NamedRerouteUsage()) {speedX=0.05, speedY=0.1525}) {
   parameterName="Distortion Texture 2 (subtract)"
   samplerType=SAMPLERTYPE_LinearGrayscale
   texture=/BRCosmetics/Effects/Fort_Effects/BluGlow/Violentium/Cellular_basecolor
  }
  n15 = ScalarParameter() {defaultValue=0.25, parameterName=Noise Str 2}
  n16 = TextureSample(UVs=Panner(Coordinate=TextureCoordinate() {vTiling=0.01}, Time=NamedRerouteUsage()) {speedY=0.1}) {
   samplerType=SAMPLERTYPE_LinearGrayscale
   texture=/Game/Characters/Player/Female/Medium/Bodies/F_MED_Cube_Queen/Textures/TechArt/T_CubeQueenNoiseA
  }
  n17 = Add(
   A=Subtract(A=Add(A=@n13, B=@n10), B=Multiply(A=@n14.R, B=@n15))
   B=Multiply(A=@n16.R, B=ScalarParameter() {defaultValue=0.25, parameterName=verticla texture Offset})
  )
  n18 = Power(Base=@n17) {constExponent=1.2}
  n19 = DynamicParameter() {paramNames [ErosionMult, ErosionPower, TextureOffsetX, TextureOffsetY]}
  n20 = Multiply(A=Subtract(A=@n17, B=Multiply(A=@n18)), B=@n19.ErosionMult) {constB=5}
  n21 = LinearInterpolate(
   A=@n17
   B=Saturate(None=Power(Base=@n20, Exp=@n19.ErosionPower) {constExponent=1.5})
   Alpha=ScalarParameter() {parameterName=UseFlameLicks, sliderMax=1}
  )
  n22 = MaterialFunctionCall(UVs=AppendVector(A=Add(A=@n4, B=@n10), B=Add(A=@n12, B=@n21)), "Rotation Angle (0-1)"=@n1) {materialFunction=&CustomRotator}
  n23 = ScalarParameter() {parameterName=Distortion affect Glow}
  n24 = ScalarParameter() {parameterName=Glow Texture Offset X}
  n25 = ScalarParameter() {parameterName=Glow Texture Offset Y}
  n26 = TextureSampleParameter2D(UVs=Add(A=LinearInterpolate(A=@n2.Rotated Values, B=@n22.Rotated Values, Alpha=@n23), B=AppendVector(A=@n24, B=@n25))) {
   parameterName=Glow Texture
   samplerSource=SSM_Clamp_WorldGroupSettings
   samplerType=SAMPLERTYPE_LinearGrayscale
   texture=&/T_Dark_PIt_Back_Fill_mask
  }
  MaterialFunctionCall(Emissive=Multiply(A=VectorParameter() {parameterName=Glow Color, defaultValue=<rgba 0.25 0.6453 1 1>}.RGB, B=@n26.RGB)) {materialFunction=&MF_FX_EmissiveAdjustments}
  n27 = TextureSampleParameter2D(UVs=@n22.Rotated Values) {
   parameterName=Main Texture
   samplerSource=SSM_Clamp_WorldGroupSettings
   samplerType=SAMPLERTYPE_LinearGrayscale
   texture=&T_Dark_PIt_Back_Fill_mask
  }
  n28 = Power(Base=OneMinus(None=MaterialFunctionCall() {materialFunction=&LinearGradient}.VGradient)) {constExponent=150}
  n29 = Multiply(A=Multiply(A=ParticleColor().A, B=@n27.R), B=OneMinus(None=Clamp(None=@n28)))
  n30 = MaterialFunctionCall("FadeSensitivity (optional)"=ScalarParameter() {group=AngleFade, parameterName=FadeSensitivity, sliderMax=1}) {materialFunction=&FadeBasedOnViewAngle}
  n31 = Constant() {r=1}
  n32 = ScalarParameter() {parameterName=DepthFadeDist}
  Multiply(
   A=LinearInterpolate(A=@n29, B=Multiply(A=@n29, B=@n30.Result), Alpha=ScalarParameter() {group=AngleFade, parameterName=UseFadeBasedOnAngle, sliderMax=1})
   B=ShadingPathSwitch(Default=DepthFade(Opacity=@n31, FadeDistance=@n32), Mobile=Constant() {r=1}) {desc=Migrated from FeatureLevelSwitch}
  )
  n33 = Multiply(A=Time(), B=ScalarParameter() {defaultValue=0.5, parameterName=Master Time})
  NamedRerouteDeclaration(None=@n33) {name=Master Time RR, nodeColor=<rgba 1 0 0.8672 1>}
  Reroute(None=@n33)
 ]
}
