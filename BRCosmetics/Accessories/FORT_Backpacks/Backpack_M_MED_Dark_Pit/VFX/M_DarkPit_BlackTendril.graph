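# Masked, two-sided Default Lit material for Niagara ribbon tendrils on the Dark Pit backpack.
# Panning UV1 noise, shaped by power, scale, bias and contrast from dynamic parameters, distorts the UV0 tendril and glow texture lookups.
# Tendril, noise, particle alpha and a sphere mask around the offset object position multiply into alpha, depth-faded except on mobile.
# The tinted glow, scaled by a dynamic parameter, feeds an emissive adjustment function.
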Material /BRCosmetics/Accessories/FORT_Backpacks/Backpack_M_MED_Dark_Pit/VFX/M_DarkPit_BlackTendril
graph {
 properties {bUsedWithNiagaraRibbons=true, blendMode=BLEND_Masked, pixelDepthOffsetMode=PDOM_Legacy, twoSided=true, usedShadingModels=Default Lit}
 nodes [
  ActorPositionWS()
  Constant() {r=1}
  Constant()
  n1 = VectorParameter() {parameterName=Bottom Color Tint, defaultValue=<rgba 0 1 1 0>}
  n2 = DynamicParameter() {desc=1, parameterIndex=1, defaultValue=<rgba 0 0 0.5 0>, paramNames [UV1 Offset X, UV1 Offset Y, Glow UV_X, Glow Str]}
  n3 = ComponentMask(None=TextureCoordinate()) {g=true}
  n4 = DynamicParameter() {desc=0, defaultValue=<rgba 1 0 1 1>, paramNames [UV1 Panner Time, UV0Shape, UV1 Shape, UV1 Tex affects UV0 Coords Str]}
  n5 = DynamicParameter() {desc=0, defaultValue=<rgba 1 0 1 1>, paramNames [UV1 Panner Time, UV0Shape, UV1 Shape, UV1 Tex affects UV0 Coords Str]}
  n6 = AppendVector(A=Constant(), B=@n5.UV1 Shape)
  n7 = TextureCoordinate() {coordinateIndex=1}
  n8 = AppendVector(A=ScalarParameter() {defaultValue=1, parameterName=UV1 Scale X}, B=ScalarParameter() {defaultValue=1, parameterName=UV1 Scale Y})
  n9 = Panner(
   Coordinate=MaterialFunctionCall(
    "Animation  Phase (0-1)"=Constant()
    Number of Rows=ScalarParameter() {defaultValue=1, group=Texture, parameterName=Num Rows 1}
    Number of Columns=ScalarParameter() {defaultValue=1, group=Texture, parameterName=Num Columns 1}
    UVs=Multiply(A=@n7, B=@n8)
   ) {materialFunction=&FlipBook}.UVs
   Time=@n5.UV1 Panner Time
   Speed=AppendVector(A=ScalarParameter() {defaultValue=0.1, parameterName=UV1 Speed X}, B=ScalarParameter() {parameterName=UV1 Speed Y})
  )
  n10 = DynamicParameter() {desc=1, parameterIndex=1, defaultValue=<rgba 0 0 0.5 0>, paramNames [UV1 Offset X, UV1 Offset Y, Glow UV_X, Glow Str]}
  n11 = TextureSampleParameter2D(UVs=Add(A=Add(A=@n6, B=@n9), B=AppendVector(A=@n10.UV1 Offset X, B=@n10.UV1 Offset X))) {
   group=Texture
   parameterName=UV1 Texture
   samplerType=SAMPLERTYPE_LinearGrayscale
   texture=/BRCosmetics/Athena/Effects/Particles/Skydiving/NS_Contrail_DarkAndLight/T_DarkAndLight_Noise
  }
  n12 = DynamicParameter() {desc=2, parameterIndex=2, defaultValue=<rgba 2 2 0.144 0.696>, paramNames [Power,Mul,Add,Contrast]}
  n13 = Power(
   Base=ChannelMaskParameter(None=@n11.RGBA) {group=Texture, maskChannel=Green, parameterName=UV1Channel, defaultValue=<rgba 0 1 0 0>}.R
   Exp=@n12.Power
  )
  n14 = MaterialFunctionCall(In=Add(A=Multiply(A=@n13, B=@n12.Mul), B=@n12.Add), Contrast=@n12.Contrast) {materialFunction=&CheapContrast}
  n15 = Multiply(A=@n4.UV1 Tex affects UV0 Coords Str, B=ConstantBiasScale(None=@n14.Result) {bias=-0.5, scale=2})
  n16 = MaterialFunctionCall(Float2=Add(A=AppendVector(A=@n2.Glow UV_X, B=@n3), B=@n15)) {materialFunction=&BreakOutFloat2Components}
  n17 = OneMinus(None=@n16.R)
  n18 = ScalarParameter() {defaultValue=1, parameterName=UV0 Texture 2 Scale X}
  n19 = ScalarParameter() {defaultValue=1, parameterName=UV0 Texture 2 Scale Y}
  n20 = MaterialFunctionCall(Float2=Multiply(
    A=MaterialFunctionCall(X=@n17, Y=@n16.G) {materialFunction=&MakeFloat2}.Result
    B=AppendVector(A=@n18, B=@n19)
   )) {materialFunction=&BreakOutFloat2Components}
  n21 = OneMinus(None=@n20.R)
  n22 = ScalarParameter() {parameterName=UV0 Texture 2 Offset X}
  n23 = ScalarParameter() {parameterName=UV0 Texture 2 Offset Y}
  n24 = TextureSampleParameter2D(UVs=Add(
    A=MaterialFunctionCall(X=@n21, Y=@n20.G) {materialFunction=&MakeFloat2}.Result
    B=AppendVector(A=@n22, B=@n23)
   )) {
   group=Texture
   parameterName=UV0 Texture 2
   samplerSource=SSM_Clamp_WorldGroupSettings
   samplerType=SAMPLERTYPE_LinearGrayscale
   texture=&/T_Dark_PIt_Ribbon_Glow
  }
  MaterialFunctionCall(Emissive=Multiply(A=Multiply(A=@n1.RGB, B=@n24.RGB), B=@n2.Glow Str)) {materialFunction=&MF_FX_EmissiveAdjustments}
  ParticleRandom()
  n25 = ParticleColor()
  Reroute(None=@n25.RGB)
  n26 = AppendVector(A=Constant(), B=@n4.UV0Shape)
  n27 = MaterialFunctionCall(
   "Animation  Phase (0-1)"=Constant()
   Number of Rows=ScalarParameter() {defaultValue=1, group=Texture, parameterName=Num Rows 0}
   Number of Columns=ScalarParameter() {defaultValue=1, group=Texture, parameterName=Num Columns 0}
   UVs=TextureCoordinate()
  ) {materialFunction=&FlipBook}
  n28 = TextureSampleParameter2D(UVs=Add(A=Add(A=@n26, B=@n27.UVs), B=@n15)) {
   group=Texture
   parameterName=UV0 Texture
   samplerSource=SSM_Clamp_WorldGroupSettings
   texture=&/T_DarkPitTendrils_Single
  }
  n29 = ChannelMaskParameter(None=@n28.RGBA) {group=Texture, parameterName=UV0Channel}
  n30 = ObjectPositionWS()
  n31 = Transform(None=VectorParameter() {group=SphereMask, parameterName=Actor Position Offset}.RGB) {transformSourceType=TRANSFORMSOURCE_Local}
  n32 = Multiply(
   A=Multiply(A=@n25.A, B=Multiply(A=@n29.R, B=@n14.Result))
   B=SphereMask(
    A=WorldPosition().XYZ
    B=Add(A=@n30, B=@n31)
    Radius=ScalarParameter() {defaultValue=25, group=SphereMask, parameterName=Radius}
    Hardness=ScalarParameter() {defaultValue=0.05, group=SphereMask, parameterName=Hardness}
   )
  )
  Reroute(None=@n32)
  n33 = ScalarParameter() {defaultValue=50, group=Alpha, parameterName=FadeDistance, sortPriority=2}
  Reroute(None=ShadingPathSwitch(Default=DepthFade(Opacity=@n32, FadeDistance=@n33), Mobile=@n32) {desc=Migrated from FeatureLevelSwitch})
  VectorParameter() {group=Base, parameterName=SpecMetalRough}
 ]
}
