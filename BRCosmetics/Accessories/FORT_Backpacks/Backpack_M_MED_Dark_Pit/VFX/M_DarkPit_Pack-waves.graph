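# Additive, unlit, two-sided material for Niagara mesh particles, drawing the wave effect on the Dark Pit backpack.
# Two panning smoke-noise samples are blended, sharpened and faded along V, then added to a V gradient; dynamic parameters drive Y tiling and pan speed.
# Particle alpha scales the resulting mask, with depth fade bypassed on mobile.
# Particle color feeds a shared emissive adjustment function.
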Material /BRCosmetics/Accessories/FORT_Backpacks/Backpack_M_MED_Dark_Pit/VFX/M_DarkPit_Pack-waves
graph {
 properties {
  bUsedWithNiagaraMeshParticles=true
  blendMode=BLEND_Additive
  pixelDepthOffsetMode=PDOM_Legacy
  shadingModel=MSM_Unlit
  twoSided=true
  usedShadingModels=Unlit
 }
 nodes [
  n1 = ParticleColor()
  MaterialFunctionCall(Emissive=@n1.RGB) {materialFunction=&MF_FX_EmissiveAdjustments}
  n2 = ComponentMask(None=TextureCoordinate()) {g=true}
  n3 = Multiply(
   A=ScalarParameter() {defaultValue=1, parameterName=Grad-Add}
   B=Power(Base=@n2, Exp=ScalarParameter() {defaultValue=2, parameterName=Grad-Add_Exponent}) {constExponent=6}
  )
  n4 = Power(Base=@n2, Exp=ScalarParameter() {defaultValue=2, parameterName=NoiseFadePower})
  n5 = ScalarParameter() {defaultValue=8, parameterName=NoiseFadeMult}
  n6 = AppendVector(A=ScalarParameter() {defaultValue=0.2, parameterName=NoiseA_SpeedX}, B=ScalarParameter() {defaultValue=1, parameterName=NoiseA_SpeedY})
  n7 = DynamicParameter() {paramNames [YTile,PanSpeed,Param3,Param4]}
  n8 = Multiply(A=Time(), B=@n7.PanSpeed)
  n9 = AppendVector(A=ScalarParameter() {defaultValue=1, parameterName=X_Tile}, B=@n7.YTile)
  n10 = Multiply(A=TextureCoordinate() {vTiling=0.5}, B=@n9)
  n11 = Multiply(A=@n9, B=TextureCoordinate() {vTiling=0.2})
  n12 = AppendVector(A=ScalarParameter() {defaultValue=-0.25, parameterName=NoiseB_SpeedX}, B=ScalarParameter() {defaultValue=0.8, parameterName=NoiseB_SpeedY})
  n13 = LinearInterpolate(
   A=TextureSampleParameter2D(UVs=Panner(Coordinate=@n10, Time=@n8, Speed=@n6) {speedX=0.2, speedY=1}) {parameterName=NoiseA, texture=&T_Inky_Smoke_Tile_Chromatic}.RGB
   B=TextureSampleParameter2D(UVs=Panner(Coordinate=@n11, Time=@n8, Speed=@n12) {speedX=-0.25, speedY=0.8}) {parameterName=NoiseB, texture=&T_Inky_Smoke_Tile_Chromatic}.RGB
  )
  n14 = ScalarParameter() {defaultValue=2.5, parameterName=Noise_Exponent}
  n15 = Multiply(
   A=Clamp(None=Multiply(A=@n4, B=@n5))
   B=Multiply(A=Power(Base=@n13, Exp=@n14), B=ScalarParameter() {defaultValue=15, parameterName=Noise_Mult}) {constB=1.5}
  )
  n16 = Constant() {r=1}
  n17 = ScalarParameter() {parameterName=DepthFadeDist}
  Multiply(
   A=Multiply(A=@n1.A, B=Add(A=@n3, B=@n15))
   B=ShadingPathSwitch(Default=DepthFade(Opacity=@n16, FadeDistance=@n17), Mobile=Constant() {r=1}) {desc=Migrated from FeatureLevelSwitch}
  )
  Reroute(None=@n1.RGB)
 ]
}
