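# Niagara effect for the Dark Pit backpack cosmetic: a dark orb with blue glow, light, flames, waves and rising tendrils.
# Mesh emitters build the orb, center, flames and waves, with glow sprites alongside.
# Two ribbon emitters launch curl-noise or drag-slowed tendrils from sockets on the user skeletal mesh, widened by a system camera scalar.
# The orb adjusts to view direction, and flame and glow vary with time of day.
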NiagaraSystem /BRCosmetics/Accessories/FORT_Backpacks/Backpack_M_MED_Dark_Pit/VFX/NS_Backpack_DarkPit
dependencies {
 usedDataInterfaces [
  &NiagaraDataInterfaceCamera
  &NiagaraDataInterfaceCurve
  &NiagaraDataInterfaceArrayFloat
  &NiagaraDataInterfacePlatformSet
  &NiagaraDataInterfaceMeshRendererInfo
 ]
 usedDynamicInputs [&NormalizeDistanceRange]
 usedModules [
  &SystemState
  &self:SystemSpawnScript.SystemScriptSource.SystemScriptGraph.NiagaraNodeAssignment_1.SetVariables_#1
 ]
}
emitters {
 Center001 {
  emitterUpdateScript [
   &EmitterState()
   &SpawnRate(SpawnRate=0.7, Use Spawn Probability=false)
   &SpawnBurst_Instantaneous(Use Spawn Probability=false, UseLoopCountLimit=false)
  ]
  particleSpawnScript [
   &InitializeParticle(
    Color=<rgba 0 0 0 15>
    Lifetime Max=4.25
    Lifetime Min=3.4
    Lifetime Mode=::NewEnumerator1
    Mesh Scale Mode=::NewEnumerator1
    Mesh Uniform Scale Max=0.13
    Mesh Uniform Scale Min=0.117
    UsePositionOffset=false
    Write Material Random=true
   )
   &InitialMeshOrientation(Use Rotation Vector=false)
  ]
  particleUpdateScript [
   &ParticleState()
   &ScaleMeshSize(Scale Factor=&VectorFromFloat(Value=&FloatFromCurve(FloatCurve=DataInterface() {propertyValues {Curve {keys [
          {interpMode=RCIM_Cubic, leaveTangentWeight=0.1667, tangentMode=RCTM_User, tangentWeightMode=RCTWM_WeightedBoth, time=0, value=0}
          {
           arriveTangentWeight=0.5697
           interpMode=RCIM_Cubic
           leaveTangentWeight=0.608
           tangentMode=RCTM_Break
           tangentWeightMode=RCTWM_WeightedBoth
           time=0.5
           value=1
          }
          {
           arriveTangent=-0.2247
           arriveTangentWeight=0.076
           interpMode=RCIM_Cubic
           leaveTangent=-0.2247
           leaveTangentWeight=0.076
           tangentMode=RCTM_User
           tangentWeightMode=RCTWM_WeightedBoth
           time=1
           value=0
          }
         ]}}})))
   &ScaleColor(
    Scale Alpha=&FloatFromCurve(FloatCurve=DataInterface() {propertyValues {Curve {keys [<at 0 1>,<at 0.5 1>,<at 1 0>]}}})
    ScaleA=true
    ScaleRGB=true
   ) {enabled=false}
   &UpdateMeshOrientation(
    Rotation Rate=&RandomRangeFloat(Minimum=0, Maximum=0.0402)
    Rotation Vector=&RandomVector(Vector Scale=1)
   )
  ]
  rendererClasses [&NiagaraSpriteRendererProperties, &NiagaraMeshRendererProperties]
  renderers [
   {rendererClass=&NiagaraSpriteRendererProperties}
   {rendererClass=&NiagaraMeshRendererProperties, rendererIndex=1}
  ]
 }
 Empty {
  emitterUpdateScript [
   &EmitterState()
   &SpawnBurst_Instantaneous(Age:NiagaraFloat=$Emitter.Age, Use Spawn Probability=false, UseLoopCountLimit=false)
  ]
  particleSpawnScript [
   &InitializeParticle(
    Color=&FN_Select_Color_from_Time_Of_Day(Morning=<rgba 0.056 0.5536 1 0.025>, Day=<rgba 0.056 0.5536 1 0>, Evening=<rgba 0.056 0.5536 1 0.025>, Night=<rgba 0.056 0.5536 1 0.018>)
    Lifetime=5
    Sprite Size=<25 27>
    Sprite Size Mode=::NewEnumerator4
    UsePositionOffset=false
    Write Material Random=true
   )
   &InitializeParticle(Color=<rgba 0.056 0.5536 1 50>, Lifetime=5, UsePositionOffset=false, Write Material Random=true) {enabled=false, module=InitializeParticle001}
   &DynamicMaterialParameters(
    Index 0 Param 1=1
    Index 0 Param 2=0.25
    Index 0 Param 3=1
    Index 0 Param 4=0.1
    Module.Param0WriteEnabled=true
    Module.Param0WriteEnabled1=true
    Module.Param0WriteEnabled3=true
    Module.Param1WriteEnabled=true
    Module.Param1WriteEnabled1=true
    Module.Param1WriteEnabled3=true
    Module.Param2WriteEnabled=true
    Module.Param2WriteEnabled1=true
    Module.Param2WriteEnabled3=true
    Module.Param3WriteEnabled=true
    Module.Param3WriteEnabled1=true
    Module.Param3WriteEnabled3=true
    Write Parameter Index 1=true
    Write Parameter Index 3=true
   )
   &CameraOffset(Camera Offset Amount=5)
   /SetVariables(Particles.Air Glow Color:LinearColor=<rgba 0.056 0.5536 1 0.15>) {isSetParameters=true}
   /SetVariables(Particles.Air Glow Color:LinearColor=<rgba 0.056 0.5536 1 50>) {enabled=false, isSetParameters=true}
  ]
  particleUpdateScript [
   &ParticleState(
    Kill Particles When Lifetime Has Elapsed=false
    Let Infinitely Lived Particles Die When Emitter Deactivates=true
    Loop Particles Lifetime=true
   )
   &ScaleColor(
    Scale Alpha=&Lerp_Float(A=0.75, B=1, Alpha=&Sine(Normalized Angle=$Emitter.Age, Period=1, Scale=0.5, Bias=0.5))
    ScaleA=true
    ScaleRGB=true
   ) {enabled=false}
   &ScaleColor(
    Scale Alpha=&FN_Select_Float_from_Time_Of_Day(Morning=1, Day=1, Evening=1, Night=1)
    ScaleA=true
    ScaleRGB=true
   ) {module=ScaleColor001}
  ]
  rendererClasses [&NiagaraSpriteRendererProperties]
  renderers [
   {rendererClass=&NiagaraSpriteRendererProperties}
   {rendererClass=&NiagaraSpriteRendererProperties, rendererIndex=1}
   {rendererClass=&NiagaraSpriteRendererProperties, rendererIndex=2}
  ]
 }
 Empty001 {
  bEnabled=false
  emitterUpdateScript [
   &EmitterState()
   &SpawnBurst_Instantaneous(Use Spawn Probability=false, UseLoopCountLimit=false)
  ]
  particleSpawnScript [
   &InitializeParticle(Color Mode=::NewEnumerator0, Lifetime=5, Sprite Size Mode=::NewEnumerator3, UsePositionOffset=false, Write Material Random=true)
   &CameraOffset(Camera Offset Amount=100)
  ]
  particleUpdateScript [
   &ParticleState()
   &DynamicMaterialParameters(
    Index 0 Param 1=&DotProduct(
     Vector A=$Engine.Owner.SystemZAxis
     Vector B=&ReturnCameraVector(Camera Vectors=::NewEnumerator4, Coordinate Space=::Simulation, Origin Position=$Particles.Position)
    )
    Module.Index0Param0Write=true
    Module.Index0Param1Write=true
    Module.Index0Param2Write=true
    Module.Index0Param3Write=true
   )
  ]
  rendererClasses [&NiagaraSpriteRendererProperties, &NiagaraMeshRendererProperties]
  renderers [
   {rendererClass=&NiagaraSpriteRendererProperties}
   {rendererClass=&NiagaraMeshRendererProperties, rendererIndex=1}
  ]
 }
 Fixed_Splines_Segments {
  emitterSpawnScript [/SetVariables(Emitter.NumBeams=1) {isSetParameters=true, module=SetVariables_#2}]
  emitterUpdateScript [
   &EmitterState(
    Life Cycle Mode=::NewEnumerator1
    Loop Delay=0.1
    Loop Duration=&RandomRangeFloat(Minimum=4, Maximum=4)
    Recalculate Duration Each Loop=true
    UseLoopDelay=true
   )
   &SpawnRate(SpawnRate=&Multiply_FloatByInt(Float=2, Integer=$Emitter.Initial.NumBeams), Use Spawn Probability=false)
   /SetVariables(
    Emitter.Random FLoat:NiagaraFloat=&RandomRangeFloat(Minimum=0.05, Maximum=0.125)
    Emitter.Random float 2:NiagaraFloat=&RandomRangeFloat(Minimum=0, Maximum=1)
    Emitter.RandomInt:NiagaraInt32=&RandomRangeInt(Minimum=0, Maximum=5)
    Emitter.random vel mul:NiagaraFloat=&RandomRangeFloat(Minimum=0.75, Maximum=1)
   ) {isSetParameters=true}
  ]
  particleSpawnScript [
   &SphereLocation(
    MaskSpawnGroup=false
    Override Seed=false
    Sphere Radius=1
    Surface Only=false
    UseNonUniformScale=false
    UseOffset=false
    UseSphereOrientation=false
   )
   &AddVelocityInCone(
    Cone Angle=20
    Cone Axis=<0 0 1>
    Module.Use Velocity Falloff On Cone Axis=false
    Velocity Distribution Along Cone Axis=0.75
    Velocity Strength=&Multiply_Float(A=&RandomRangeFloat(Minimum=3.45, Maximum=3.5), B=$Emitter.random vel mul)
   )
   &InitializeRibbon(
    Color=<rgba 0 0 0 10>
    Lifetime=9
    Material Random:NiagaraFloat=&SelectFloatFromArray(
     Direct Array Index=$Emitter.LoopCount
     Direct Array Mode=::NewEnumerator1
     Float Selection Array=DataInterface() {propertyValues {FloatData [0,0.25,0.5,0.75]}}
    )
    Ribbon Width=&FloatFromCurve(
     FloatCurve=DataInterface() {propertyValues {Curve {keys [
         <at 0 0 RCIM_Constant>
         {arriveTangent=4, interpMode=RCIM_Cubic, leaveTangent=4, time=0.25, value=1}
         {arriveTangent=-1.5385, interpMode=RCIM_Cubic, leaveTangent=-1.5385, time=0.9, value=0}
        ]}}}
     CurveIndex=$Emitter.NormalizedLoopAge
     Scale Curve=1.25
    )
    UsePositionOffset=false
    Write Material Random=true
   )
   &SkeletalMeshLocation(
    Bone / Socket Index=$Emitter.RandomInt
    MaskSpawnGroup=false
    Mesh Sampling Type=::NewEnumerator5
    Skeletal Mesh=$User.SkeletalMesh
    Socket Sampling Mode=::NewEnumerator5
    Use Sampled Mesh Attribute as Ribbon ID=false
   )
   /SetVariables(
    Particles.RibbonID=/Niagara/DynamicInputs/NiagaraID/MakeNiagaraIDFromInt(Index:NiagaraInt32=$Emitter.LoopCount, Acquire Tag:NiagaraInt32=0)
    Particles.RibbonLinkOrder:NiagaraFloat=&Add_Float(A=$Emitter.Age, B=-10)
   ) {isSetParameters=true}
   /SetVariables(
    Particles.GetRandom Float:NiagaraFloat=$Emitter.Random FLoat
    Particles.RibbonUVDistance:NiagaraFloat=&Multiply_Float(A=$Emitter.Age, B=10)
   ) {isSetParameters=true}
  ]
  particleUpdateScript [
   &ParticleState()
   &ScaleColor(
    Scale Alpha=&FloatFromCurve(FloatCurve=DataInterface() {propertyValues {Curve {keys [
         {arriveTangent=5.8219, interpMode=RCIM_Cubic, leaveTangent=5.8219, tangentMode=RCTM_User, time=0, value=0}
         {arriveTangent=-0.0383, interpMode=RCIM_Cubic, leaveTangent=-0.0383, tangentMode=RCTM_User, time=0.5, value=1.0113}
         {arriveTangentWeight=0.3351, interpMode=RCIM_Cubic, tangentMode=RCTM_User, time=1, value=0}
        ]}}})
    ScaleA=true
    ScaleRGB=true
   )
   &CurlNoiseForce(Noise Frequency=120, Noise Quality / Cost=::NewEnumerator0, Noise Strength=0.25, Pan Noise=false)
   &InheritSourceMovement(ApplyForce=false, ApplyPosition=true, ApplyPositionFalloff=false)
   &SolveForcesAndVelocity(Clamp Velocity=false, Limit Acceleration=false)
   /SetVariables(Particles.From Center Pos:Vector3f=&Lerp_Vector(
     A=/Niagara/DynamicInputs/Helpers/SimulationPosition()
     B=Linked() {linkedVariable {name=Particles.Position, type=NiagaraPosition}}
     Alpha=&FloatFromCurve(FloatCurve=DataInterface() {propertyValues {Curve {keys [
          {interpMode=RCIM_Cubic, leaveTangentWeight=0.3399, tangentMode=RCTM_User, time=0, value=0}
          {arriveTangentWeight=0.3399, interpMode=RCIM_Cubic, tangentMode=RCTM_User, time=0.5, value=1}
         ]}}})
    )) {isSetParameters=true, module=SetVariables_#3}
   &DynamicMaterialParameters(
    Index 0 Param 1=&Multiply_Float(A=$Engine.Time, B=$Particles.Initial.GetRandom Float)
    Module.Param0WriteEnabled=true
    Module.Param1WriteEnabled=true
    Module.Param2WriteEnabled=true
    Module.Param3WriteEnabled=true
   )
   &ScaleRibbonWidth(Ribbon Width Scale=&Multiply_Float(
     A=&FloatFromCurve(FloatCurve=DataInterface() {propertyValues {Curve {keys [
          {arriveTangent=5, interpMode=RCIM_Cubic, leaveTangent=5, time=0, value=0}
          {arriveTangentWeight=0.1491, interpMode=RCIM_Cubic, leaveTangentWeight=0.2236, tangentMode=RCTM_User, time=0.2, value=1}
          {arriveTangentWeight=0.2075, interpMode=RCIM_Cubic, leaveTangentWeight=0.1729, tangentMode=RCTM_User, time=0.5, value=1}
          {arriveTangent=-4, interpMode=RCIM_Cubic, leaveTangent=-4, time=0.75, value=0}
         ]}}})
     B=&Lerp_Float(A=0.5, B=3, Alpha=$System.Cam Scalar)
    ))
  ]
  rendererClasses [&NiagaraSpriteRendererProperties, &NiagaraRibbonRendererProperties]
  renderers [
   {rendererClass=&NiagaraSpriteRendererProperties}
   {rendererClass=&NiagaraRibbonRendererProperties, rendererIndex=1}
  ]
 }
 Fixed_Splines_local {
  emitterSpawnScript [/SetVariables(Emitter.NumBeams=6) {isSetParameters=true, module=SetVariables_#2}]
  emitterUpdateScript [
   &EmitterState(Life Cycle Mode=::NewEnumerator1, Loop Duration=15, UseLoopDelay=false)
   &SpawnRate(SpawnRate=&Multiply_FloatByInt(Float=1.4, Integer=$Emitter.Initial.NumBeams), Use Spawn Probability=false)
  ]
  particleSpawnScript [
   &FN_SetSpawnGroup(
    Inclusivity=true
    Normalize Spawn Group=true
    Number of Spawn Groups:NiagaraInt32=$Emitter.Initial.NumBeams
    Particles Per Group=1
    Set Ribbon ID and Ribbon Link Order=true
   )
   &InitializeRibbon(
    Color=<rgba 0 0 0 10>
    Lifetime=&SelectFloatFromArray(Direct Array Index=$Particles.SpawnGroup, Float Selection Array=DataInterface() {propertyValues {FloatData [0,11.9,0,12.75,9.35,0]}})
    Material Random:NiagaraFloat=&SelectFloatFromArray(Direct Array Index=$Particles.SpawnGroup, Float Selection Array=DataInterface() {propertyValues {FloatData [0,0.25,0.5,0.75]}})
    Ribbon Width=&SelectFloatFromArray(Direct Array Index=$Particles.SpawnGroup, Float Selection Array=DataInterface() {propertyValues {FloatData [6.8,2.72,2.72,2.72,2.72,2.04]}})
    UsePositionOffset=false
    Write Material Random=true
   )
   &SkeletalMeshLocation(
    Bone / Socket Index=$Particles.SpawnGroup
    Interpolate Mesh Sampling=false
    MaskSpawnGroup=false
    Mesh Sampling Type=::NewEnumerator5
    Orientation Sampling=::NewEnumerator1
    Skeletal Mesh=$User.SkeletalMesh
    Socket Sampling Mode=::NewEnumerator5
    Use Sampled Mesh Attribute as Ribbon ID=false
   )
   &AddVelocityInCone(
    Cone Angle=13
    Cone Axis=<0 0 1>
    Module.Use Velocity Falloff On Cone Axis=false
    Velocity Distribution Along Cone Axis=0.75
    Velocity Strength=&Multiply_Float(
     A=&RandomRangeFloat(Minimum=2.8, Maximum=3)
     B=&SelectFloatFromArray(Direct Array Index=$Particles.SpawnGroup, Float Selection Array=DataInterface() {propertyValues {FloatData [0,1.1,0,1.25,1.15,0]}})
    )
   )
   /SetVariables(Particles.RibbonUVDistance:NiagaraFloat=&Add_Float(
     A=&Multiply_Float(A=$Emitter.Age, B=10)
     B=&RandomRangeFloat(Minimum=-1, Maximum=1)
    )) {isSetParameters=true}
  ]
  particleUpdateScript [
   &ParticleState()
   &ScaleColor(
    Scale Alpha=&MakeCustomFloatFromBool(
     Bool=&IsPlatformSetActive(New Platform Set=DataInterface() {propertyValues {Platforms {qualityLevelMask=-29}}})
     True Float=&FloatFromCurve(FloatCurve=DataInterface() {propertyValues {Curve {keys [
          {arriveTangent=0.5472, interpMode=RCIM_Cubic, leaveTangent=0.5472, leaveTangentWeight=0.2235, tangentMode=RCTM_User, time=0, value=0}
          {arriveTangent=7.416, interpMode=RCIM_Cubic, leaveTangent=-0.7463, tangentMode=RCTM_Break, time=0.5347, value=1}
          {
           arriveTangent=-2.7374
           arriveTangentWeight=0.114
           interpMode=RCIM_Cubic
           leaveTangent=-2.7374
           leaveTangentWeight=0.228
           time=0.7347
           value=0.3
          }
          {arriveTangentWeight=0.3723, interpMode=RCIM_Cubic, tangentMode=RCTM_User, time=0.9, value=0}
         ]}}})
     False Float=&FloatFromCurve(FloatCurve=DataInterface() {propertyValues {Curve {keys [
          {interpMode=RCIM_Cubic, leaveTangentWeight=0.2235, tangentMode=RCTM_User, time=0, value=0}
          {arriveTangent=7.416, interpMode=RCIM_Cubic, leaveTangent=-0.7463, tangentMode=RCTM_Break, time=0.3246, value=1}
          {arriveTangent=-2.8, arriveTangentWeight=0.114, interpMode=RCIM_Cubic, leaveTangent=-2.8, leaveTangentWeight=0.228, time=0.5746, value=0.3}
          {arriveTangentWeight=0.3723, interpMode=RCIM_Cubic, tangentMode=RCTM_User, time=0.9, value=0}
         ]}}})
    )
    ScaleA=true
    ScaleRGB=true
   )
   &Drag(Drag=&RandomRangeFloat(Minimum=0, Maximum=0.05), Use Linear Drag=true, Use Rotational Drag=false) {module=Drag001}
   &SolveForcesAndVelocity(Clamp Velocity=false, Limit Acceleration=false)
   &ScaleRibbonWidth(Ribbon Width Scale=&Multiply_Float(
     A=&Multiply_Float(
      A=&MaskFloatBySpawnGroup(
       Masked Float=&FloatFromCurve(FloatCurve=DataInterface() {propertyValues {Curve {keys [
            {
             arriveTangentWeight=0.411
             interpMode=RCIM_Cubic
             leaveTangentWeight=0.1333
             tangentMode=RCTM_User
             tangentWeightMode=RCTWM_WeightedBoth
             time=0
             value=1.35
            }
            {arriveTangent=-1.4375, interpMode=RCIM_Cubic, leaveTangent=-1.4375, time=0.3, value=0.675}
            <at 0.8 0.2>
           ]}}})
       Passthrough Float=&FloatFromCurve(FloatCurve=DataInterface() {propertyValues {Curve {keys [
            {
             arriveTangent=53.2303
             arriveTangentWeight=0.5628
             interpMode=RCIM_Cubic
             leaveTangent=53.2303
             leaveTangentWeight=0.5628
             tangentMode=RCTM_User
             tangentWeightMode=RCTWM_WeightedBoth
             time=0
             value=0.25
            }
            {arriveTangent=0.7375, interpMode=RCIM_Cubic, leaveTangent=0.7375, tangentMode=RCTM_User, time=0.101, value=1.0639}
            {arriveTangent=-1.2359, interpMode=RCIM_Cubic, leaveTangent=-1.2359, time=0.4, value=0.9}
            <at 0.8 0.75>
           ]}}})
      )
      B=1
     )
     B=&Lerp_Float(A=1, B=1.25, Alpha=$System.Cam Scalar)
    ))
   /SetVariables(Particles.From Center Pos:Vector3f=&Lerp_Vector(
     A=&SimulationPosition()
     B=Linked() {linkedVariable {name=Particles.Position, type=NiagaraPosition}}
     Alpha=&FloatFromCurve(FloatCurve=DataInterface() {propertyValues {Curve {keys [
          {interpMode=RCIM_Cubic, leaveTangentWeight=0.3399, tangentMode=RCTM_User, time=0, value=0}
          {arriveTangentWeight=0.3399, interpMode=RCIM_Cubic, tangentMode=RCTM_User, time=0.35, value=1}
         ]}}})
    )) {isSetParameters=true, module=SetVariables_#3}
   &DynamicMaterialParameters(
    Index 0 Param 1=&Multiply_Float(
     A=&Add_Float(A=$Engine.Time, B=&RandomRangeFloat(Minimum=-0.1, Maximum=0.1))
     B=1
    )
    Index 0 Param 4=&FloatFromCurve(
     FloatCurve=DataInterface() {propertyValues {Curve {keys [
         {interpMode=RCIM_Cubic, leaveTangentWeight=0.4714, tangentMode=RCTM_User, time=0, value=0.25}
         {arriveTangentWeight=0.4714, interpMode=RCIM_Cubic, tangentMode=RCTM_User, time=1, value=1}
        ]}}}
     Scale Curve=0.25
    )
    Index 1 Param 1=&RandomRangeFloat(Minimum=0.1, Maximum=-0.1)
    Index 1 Param 2=&RandomRangeFloat(Minimum=&RandomRangeFloat(Minimum=-0.25, Maximum=0.25), Maximum=0)
    Index 1 Param 3=&FloatFromCurve(
     FloatCurve=DataInterface() {propertyValues {Curve {keys [{tangentMode=RCTM_User, time=0, value=1}, {tangentMode=RCTM_User, time=1, value=0}]}}}
     Scale Curve=2
    )
    Index 1 Param 4=&FloatFromCurve(
     FloatCurve=DataInterface() {propertyValues {Curve {keys [
         {arriveTangent=-11.0669, interpMode=RCIM_Cubic, leaveTangent=-11.0669, tangentMode=RCTM_User, time=0, value=25}
         {arriveTangent=-4.9472, interpMode=RCIM_Cubic, leaveTangent=-4.9472, tangentMode=RCTM_User, time=0.4, value=1}
         <at 0.864 1>
         {arriveTangentWeight=1.6997, interpMode=RCIM_Cubic, tangentMode=RCTM_User, time=1, value=0}
        ]}}}
     Scale Curve=&RandomRangeFloat(Minimum=0.25, Maximum=1)
    )
    Index 2 Param 1=&FloatFromCurve(
     FloatCurve=DataInterface() {propertyValues {Curve {keys [
         {arriveTangent=3.7115, interpMode=RCIM_Cubic, leaveTangent=3.7115, tangentMode=RCTM_User, time=0, value=0}
         {arriveTangentWeight=0.4714, interpMode=RCIM_Cubic, tangentMode=RCTM_User, time=1, value=1}
        ]}}}
     Scale Curve=2
    )
    Index 2 Param 2=10
    Index 2 Param 3=&Add_Float(
     A=&MaskFloatBySpawnGroup(
      Masked Float=&FloatFromCurve(
       FloatCurve=DataInterface() {propertyValues {Curve {keys [
           {interpMode=RCIM_Cubic, leaveTangentWeight=4.3017, tangentMode=RCTM_User, time=0, value=0}
           {arriveTangentWeight=0.3333, interpMode=RCIM_Cubic, tangentMode=RCTM_User, time=1, value=0}
          ]}}}
       Scale Curve=0.144
      )
      Passthrough Float=&FloatFromCurve(
       FloatCurve=DataInterface() {propertyValues {Curve {keys [
           {
            arriveTangent=-19.036
            arriveTangentWeight=5.2126
            interpMode=RCIM_Cubic
            leaveTangent=-19.036
            leaveTangentWeight=5.2126
            tangentMode=RCTM_User
            tangentWeightMode=RCTWM_WeightedBoth
            time=0
            value=5
           }
           {arriveTangentWeight=1.2132, interpMode=RCIM_Cubic, tangentMode=RCTM_User, time=1, value=0}
          ]}}}
       Scale Curve=0.75
      )
     )
     B=&RandomRangeFloat(Minimum=-1, Maximum=1)
    )
    Index 2 Param 4=0.696
    Module.Param0WriteEnabled=true
    Module.Param0WriteEnabled1=true
    Module.Param0WriteEnabled2=true
    Module.Param1WriteEnabled=true
    Module.Param1WriteEnabled1=true
    Module.Param1WriteEnabled2=true
    Module.Param2WriteEnabled=true
    Module.Param2WriteEnabled1=true
    Module.Param2WriteEnabled2=true
    Module.Param3WriteEnabled=true
    Module.Param3WriteEnabled1=true
    Module.Param3WriteEnabled2=true
    Write Parameter Index 1=true
    Write Parameter Index 2=true
   )
  ]
  rendererClasses [&NiagaraRibbonRendererProperties]
  renderers [{rendererClass=&NiagaraRibbonRendererProperties}]
 }
 Flame01 {
  emitterUpdateScript [
   &EmitterState(
    Inactive Response=::NewEnumerator1
    Life Cycle Mode=::NewEnumerator1
    Loop Behavior=::NewEnumerator1
    Loop Duration Mode=::NewEnumerator1
    Reset Age On Awaken=true
    Scalability Mode=::NewEnumerator1
    UseLoopDelay=false
   )
   &SpawnRate(SpawnRate=2, Use Spawn Probability=false)
  ]
  particleSpawnScript [
   &FN_SetSpawnGroup(Number of Spawn Groups=2)
   &InitializeParticle(
    Color=<rgba 0 0.7155 1 0.25>
    Lifetime Max=3.5
    Lifetime Min=2.5
    Lifetime Mode=::NewEnumerator1
    Material Random:NiagaraFloat=&RandomRangeFloat(Minimum=0, Maximum=100)
    Mesh Scale=&Multiply_Vector(
     A=&RandomRangeVector(Minimum=<0.75 0.75 0.25>, Maximum=<1 0.75 0.25>)
     B=&VectorFromFloat(Value=0.2)
    )
    Mesh Scale Mode=::NewEnumerator4
    UsePositionOffset=false
    Write Material Random=true
   ) {module=InitializeParticle001}
   &DynamicMaterialParameters(
    Index 0 Param 1=&RandomRangeFloat(Minimum=0, Maximum=0.6)
    Index 0 Param 2=1
    Index 0 Param 3=30
    Module.Param0WriteEnabled=true
    Module.Param1WriteEnabled=true
    Module.Param2WriteEnabled=true
    Module.Param3WriteEnabled=true
   )
   &InitialMeshOrientation(Mesh Orientation Mode=::NewEnumerator7, Rotation=<0 0 -0.25>, Use Rotation Vector=true)
   &InitialMeshOrientation(
    Mesh Orientation Mode=::NewEnumerator7
    Rotation=&RandomRangeVector(Minimum=<0.01 0.01 0.01>, Maximum=<-0.01 -0.01 -0.01>)
    Use Rotation Vector=true
   ) {module=InitialMeshOrientation001}
   &InitialMeshOrientation(Mesh Orientation Mode=::NewEnumerator7, Rotation=<0 0.25 0>, Use Rotation Vector=true) {module=InitialMeshOrientation002}
   &SphereLocation(
    MaskSpawnGroup=false
    Offset=<0 0 -1>
    Sphere Distribution:ENiagaraSphereDistributionMode=::NewEnumerator1
    Sphere Radius=7
    U Position:NiagaraFloat=&RandomRangeFloat(Minimum=0, Maximum=1)
    UseNonUniformScale=false
    UseOffset=true
    UseSphereOrientation=false
    V Position:NiagaraFloat=&RandomRangeFloat(Minimum=0.9, Maximum=1)
   )
   &AddVelocity(Velocity=<0 0 1>)
   &CameraOffset(Camera Offset Amount=3)
   /SetVariables(Particles.Scale=&Multiply_Vector(
     A=$Particles.Scale
     B=&MakeCustomVectorFromBool(Bool=&RandomBool(), True Vector=<1 1 1>, False Vector=<-1 1 1>)
    )) {isSetParameters=true}
   &InitialMeshOrientation(Mesh Orientation Mode=::NewEnumerator2, Use Rotation Vector=false) {enabled=false, module=InitialMeshOrientation003}
  ]
  particleUpdateScript [
   &ParticleState()
   &InterpolateOverTime(Rate Of Change=10, Target Vector Value:NiagaraPosition=$Particles.Position, Variable Type=::NewEnumerator2)
   /SetVariables(Particles.Interpolate Dynamic=&MakeFloat4FromVectorAndFloat("Vector (RGB)"=$Output.InterpolateOverTime.Moving Average, "Float (Alpha)"=0.5)) {isSetParameters=true}
   &Drag(Drag=0.5, Use Linear Drag=true, Use Rotational Drag=false)
   &SolveForcesAndVelocity(Clamp Velocity=false, Limit Acceleration=false)
   &ScaleColor(
    Scale Alpha=&Multiply_Float(
     A=&FNTimeOfDayScalar(Night Scale=0.5, Morning Scale=0.7, Evening Scale=0.7)
     B=&FloatFromCurve(FloatCurve=DataInterface() {propertyValues {Curve {keys [
          {
           arriveTangentWeight=0.3952
           interpMode=RCIM_Cubic
           leaveTangentWeight=0.3952
           tangentMode=RCTM_User
           tangentWeightMode=RCTWM_WeightedBoth
           time=0
           value=0
          }
          {interpMode=RCIM_Cubic, tangentMode=RCTM_User, time=0.5, value=1}
          {arriveTangentWeight=0.1667, interpMode=RCIM_Cubic, tangentMode=RCTM_User, tangentWeightMode=RCTWM_WeightedBoth, time=1, value=0}
         ]}}})
    )
    ScaleA=true
    ScaleRGB=true
   )
   &DynamicMaterialParameters(
    Index 0 Param 1=&RandomRangeFloat(Minimum=0, Maximum=1)
    Index 0 Param 2=&RandomRangeFloat(Minimum=0, Maximum=1)
    Index 0 Param 3=1
    Index 0 Param 4=&RandomRangeFloat(Minimum=0, Maximum=1)
    Index 1 Param 1=1
    Index 1 Param 2=1
    Index 1 Param 3=1
    Index 1 Param 4=1
    Index 2 Param 1=0.45
    Module.Param0WriteEnabled=true
    Module.Param0WriteEnabled1=true
    Module.Param0WriteEnabled2=true
    Module.Param1WriteEnabled=true
    Module.Param1WriteEnabled1=true
    Module.Param1WriteEnabled2=false
    Module.Param2WriteEnabled=true
    Module.Param2WriteEnabled1=true
    Module.Param2WriteEnabled2=false
    Module.Param3WriteEnabled=true
    Module.Param3WriteEnabled1=true
    Module.Param3WriteEnabled2=false
    Write Parameter Index 1=true
    Write Parameter Index 2=true
   ) {module=DynamicMaterialParameters001}
   &ScaleMeshSize(Scale Factor=&MakeVector(
     X=&FloatFromCurve(FloatCurve=DataInterface() {propertyValues {Curve {keys [<at 0 1>,<at 1 1.5>]}}})
     Y=&FloatFromCurve(FloatCurve=DataInterface() {propertyValues {Curve {keys [<at 0 1>,<at 1 1.35>]}}})
     Z=1
    ))
  ]
  rendererClasses [&NiagaraMeshRendererProperties]
  renderers [{rendererClass=&NiagaraMeshRendererProperties, rendererIndex=0}]
 }
 Orb {
  emitterUpdateScript [
   &EmitterState()
   &SpawnBurst_Instantaneous(Age:NiagaraFloat=$Emitter.Age, Use Spawn Probability=false, UseLoopCountLimit=false)
   /SetVariables(Emitter.Camera forward:Vector3f=&ReturnCameraVector(Camera Property=::NewEnumerator0, Coordinate Space=::Simulation)) {isSetParameters=true}
  ]
  particleSpawnScript [
   &InitializeParticle(
    Color=<rgba 0 0 0 1>
    Lifetime=5
    Mesh Scale Mode=::NewEnumerator3
    Mesh Uniform Scale=0.35
    UsePositionOffset=true
    Write Material Random=true
   )
   &InitialMeshOrientation(Mesh Orientation Mode=::NewEnumerator7, Rotation=<0 0 0.5>, Use Rotation Vector=true)
   &DynamicMaterialParameters(
    Index 0 Param 2=1
    Index 0 Param 3=1
    Module.Param0WriteEnabled=true
    Module.Param1WriteEnabled=true
    Module.Param2WriteEnabled=true
    Module.Param3WriteEnabled=true
   )
   /SetVariables(Particles.New Vector 4:Vector4f=&MakeVector4(X=0, Y=1, Z=1, W=0.7902)) {isSetParameters=true}
   /SetVariables() {isSetParameters=true}
   /SetVariables(Particles.New Quat 4f:Quat4f=&MakeQuatFromAxisAngle(Axis=<0 0 1>, AngleInDegrees=120)) {isSetParameters=true}
   &CameraOffset(Camera Offset Amount=3) {module=CameraOffset001}
  ]
  particleUpdateScript [
   &ParticleState(Kill Particles When Lifetime Has Elapsed=false, Let Infinitely Lived Particles Die When Emitter Deactivates=true)
   &UpdateMeshOrientation(
    Facing Direction=&Lerp_Vector(
     A=&Multiply_Vector(
      A=&ReturnCameraVector(Camera Vectors=::NewEnumerator4, Coordinate Space=::Simulation, Origin Position=$Particles.Position)
      B=<1 1 0>
     )
     B=&ReturnCameraVector(Camera Vectors=::NewEnumerator4, Coordinate Space=::Simulation, Origin Position=$Particles.Position)
     Alpha=0.5
    )
    Facing Interpolation=false
    Orientation Method=::NewEnumerator1
    Override Quat=false
    TransformFacing=false
    TransformUp=false
   )
   /SetVariables(
    Particles.Scaled Alpha:LinearColor=&MakeLinearColor(
     R:NiagaraFloat=0
     G:NiagaraFloat=0
     B:NiagaraFloat=0
     A:NiagaraFloat=&Lerp_Float(
      A=0
      B=1
      Alpha=&FloatFromCurve(
       FloatCurve=DataInterface() {propertyValues {Curve {keys [<at -1 0>,<at -0.75 1>,<at 0.7 1>,<at 0.9 0>]}}}
       CurveIndex=&DotProduct(
        Vector A=&NormalizeVector(
         Vector To Normalize=&Subtract_Vector(A=$Emitter.Camera forward, B=&SimulationPosition())
         Safe Normalize=true
         Safe Fallback Threshold=0
         Safe Fallback Vector=<0 0 0.001>
        )
        Vector B=<0 0 1>
       )
      )
     )
    )
    Particles.Scaled Size:Vector3f=&Lerp_Vector(
     A=<0.5 0.35 0.25>
     B=$Particles.Initial.Scale
     Alpha=&FloatFromCurve(
      FloatCurve=DataInterface() {propertyValues {Curve {keys [<at -1 1>,<at -0.5 1>,<at 0.5 1>,<at 1 0>]}}}
      CurveIndex=&DotProduct(
       Vector A=&NormalizeVector(
        Vector To Normalize=&Subtract_Vector(A=$Emitter.Camera forward, B=&SimulationPosition())
        Safe Normalize=true
        Safe Fallback Threshold=0
        Safe Fallback Vector=<0 0 0.001>
       )
       Vector B=<0 0 1>
      )
     )
    )
   ) {isSetParameters=true}
   &CameraOffset(Camera Offset Amount=&FloatFromCurve(
     FloatCurve=DataInterface() {propertyValues {Curve {keys [
         <at -1 1>
         <at -0.7 1.25>
         <at -0.5 0.1>
         {arriveTangentWeight=0.1667, leaveTangentWeight=0.7425, time=0, value=0}
         <at 0.45 0.1>
         <at 0.54 0.1>
         {arriveTangent=38.0532, leaveTangent=38.0532, time=0.6, value=0.1}
         {arriveTangent=1.8, leaveTangent=1.8, time=0.7, value=1.25}
         {arriveTangent=2.4834, time=1, value=1}
        ]}}}
     CurveIndex=&DotProduct(
      Vector A=$Engine.Owner.SystemZAxis
      Vector B=&ReturnCameraVector(Camera Vectors=::NewEnumerator4, Coordinate Space=::Simulation, Origin Position=$Particles.Position)
     )
     Scale Curve=25
    )) {enabled=false}
   &FN_Light_Attributes(
    Color=<rgba 0.056 0.5536 1 15>
    Create Color Attribute=true
    Create Exponent Attribute=true
    Create Position Attribute=true
    Create Radius Attribute=true
    Exponent=3
    Position=&Add_Vector(A=Linked() {linkedVariable {name=Particles.Position, type=NiagaraPosition}}, B=<-6 0 0>)
    Radius=10
   ) {module=FN_Light_Attributes001}
   /SetVariables(Particles.OtherMeshOrientation:Quat4f=&OrientMeshToVector(
     Look At Direction=&Multiply_Vector(
      A=&ReturnCameraVector(Camera Vectors=::NewEnumerator4, Coordinate Space=::Simulation, Origin Position=$Particles.Position)
      B=<1 1 0>
     )
     Reference Axis Vector=<1 0 0>
    )) {isSetParameters=true}
  ]
  rendererClasses [&NiagaraMeshRendererProperties, &NiagaraLightRendererProperties]
  renderers [
   {rendererClass=&NiagaraMeshRendererProperties, rendererIndex=0}
   {rendererClass=&NiagaraMeshRendererProperties, rendererIndex=1}
   {rendererClass=&NiagaraLightRendererProperties, rendererIndex=2}
   {rendererClass=&NiagaraMeshRendererProperties, rendererIndex=3}
  ]
 }
 Wave {
  emitterUpdateScript [
   &EmitterState(Life Cycle Mode=::NewEnumerator1, Loop Behavior=::NewEnumerator1, Loop Delay=0.05, Loop Duration=1, UseLoopDelay=false)
   &SpawnBurst_Instantaneous(Spawn Count=2, Use Spawn Probability=false)
  ]
  particleSpawnScript [
   &InitializeParticle(
    Color Maximum=<rgba 0.006 0.006 0.011 1>
    Color Minimum=<rgba 0 0 0 0.75>
    Color Mode=::NewEnumerator2
    Lifetime=2
    Mesh Scale Mode=::NewEnumerator3
    Mesh Uniform Scale=&RandomRangeFloat(Minimum=0.15, Maximum=0.2)
    UsePositionOffset=false
    Write Material Random=true
   )
   &InitialMeshOrientation(Orientation Axis=<0 0 1>, Orientation Vector=<0 0 1>, Rotation=<0 0.25 0>, Use Orientation Vector=true, Use Rotation Vector=true)
   &InitialMeshOrientation(
    Orientation Axis=<0 0 0>
    Orientation Vector=<0 0 0>
    Rotation=&RandomRangeVector(Minimum=<0 0 0>, Maximum=<1 0 0>)
    Use Orientation Vector=false
    Use Rotation Vector=true
   ) {module=InitialMeshOrientation001}
   &DynamicMaterialParameters(
    Index 0 Param 1=1
    Index 0 Param 2=0.23
    Module.Param0WriteEnabled=true
    Module.Param1WriteEnabled=true
    Module.Param2WriteEnabled=true
    Module.Param3WriteEnabled=true
   )
  ]
  particleUpdateScript [&ParticleState(Kill Particles When Lifetime Has Elapsed=false)]
  rendererClasses [&NiagaraMeshRendererProperties]
  renderers [{rendererClass=&NiagaraMeshRendererProperties, rendererIndex=0}]
 }
}
summary {
 systemName=NS_Backpack_DarkPit
 emitters [
  {emitterName=Orb, rendererClasses [&NiagaraMeshRendererProperties, &NiagaraLightRendererProperties]}
  {
   emitterName=Fixed_Splines_Segments
   rendererClasses [&NiagaraSpriteRendererProperties, &NiagaraRibbonRendererProperties]
  }
  {emitterName=Center001, rendererClasses [&NiagaraSpriteRendererProperties, &NiagaraMeshRendererProperties]}
  {emitterName=Wave, rendererClasses [&NiagaraMeshRendererProperties]}
  {emitterName=Fixed_Splines_local, rendererClasses [&NiagaraRibbonRendererProperties]}
  {emitterName=Flame01, rendererClasses [&NiagaraMeshRendererProperties]}
  {emitterName=Empty, rendererClasses [&NiagaraSpriteRendererProperties]}
  {
   bEnabled=false
   emitterName=Empty001
   rendererClasses [&NiagaraSpriteRendererProperties, &NiagaraMeshRendererProperties]
  }
 ]
 userVariables [
  User.bIsFrontEndPreview {type=NiagaraBool}
  User.Color_01 {type=LinearColor, description="", defaultValue {struct=LinearColor, value=<rgba 1 1 1 1>}}
  User.SimRT {
   type=Object
   description=""
   defaultValue {struct=NiagaraExt_VariableValue_Object, value {object=null, objectClass=&Object}}
  }
  User.SkeletalMesh {
   type=NiagaraDataInterfaceSkeletalMesh
   defaultValue {
    struct=NiagaraExt_VariableValue_DataInterface
    value {dataInterface=&self:NiagaraDataInterfaceSkeletalMesh_0}
   }
  }
  User.SKM_02 {
   type=NiagaraDataInterfaceSkeletalMesh
   defaultValue {
    struct=NiagaraExt_VariableValue_DataInterface
    value {dataInterface=&self:NiagaraDataInterfaceSkeletalMesh_1}
   }
  }
 ]
}
